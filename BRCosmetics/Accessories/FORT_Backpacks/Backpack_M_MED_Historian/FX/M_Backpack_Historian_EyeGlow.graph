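# Additive unlit material for the eye glow effect on the Historian backpack cosmetic, used on Niagara mesh particles.
# Color parameter times particle color feeds MF_FX_EmissiveAdjustments for emissive.
# A separate alpha chain multiplies both alphas by a cubed V-coordinate gradient, then blends in an inverted Fresnel fade that dims grazing angles.
# That alpha is also softened by DepthFade, which a shading path switch replaces with a constant on mobile.
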Material /BRCosmetics/Accessories/FORT_Backpacks/Backpack_M_MED_Historian/FX/M_Backpack_Historian_EyeGlow
graph {
 properties {
  bUsedWithNiagaraMeshParticles=true
  blendMode=BLEND_Additive
  pixelDepthOffsetMode=PDOM_Legacy
  shadingModel=MSM_Unlit
  usedShadingModels=Unlit
 }
 nodes [
  n1 = VectorParameter() {parameterName=Color, defaultValue=<rgba 1 1 1 1>}
  n2 = ParticleColor()
  MaterialFunctionCall(Emissive=Multiply(A=@n1.RGB, B=@n2.RGB)) {materialFunction=&MF_FX_EmissiveAdjustments}
  n3 = TextureCoordinate()
  n4 = Multiply(A=Multiply(A=@n1.A, B=@n2.A), B=Power(Base=ComponentMask(None=@n3) {g=true}) {constExponent=3})
  n5 = Abs(None=Fresnel(ExponentIn=ScalarParameter() {defaultValue=1.5, group=CameraVectorFade, parameterName=Camera Vector Exponentln}) {baseReflectFraction=0})
  n6 = Clamp(None=Power(Base=OneMinus(None=@n5), Exp=ScalarParameter() {defaultValue=2, group=CameraVectorFade, parameterName=Camera vector power}) {constExponent=4})
  n7 = ScalarParameter() {defaultValue=10, parameterName=FadeDistance}
  Multiply(
   A=LinearInterpolate(A=@n4, B=Multiply(A=@n4, B=@n6), Alpha=ScalarParameter() {defaultValue=1, group=CameraVectorFade, parameterName=Camera vector fade})
   B=ShadingPathSwitch(Default=DepthFade(FadeDistance=@n7), Mobile=Constant() {r=1}) {desc=Migrated from FeatureLevelSwitch}
  )
 ]
}
