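# Physics asset for the Historian backpack cosmetic, defining bodies and joints for its dangling beard and rope bones.
# The beard chains capsules from a base sphere, forking at the tip, with four strands branching along it.
# A two-segment top rope and four single-segment side ropes each hang from their own small base sphere.
# Tight swing and twist limits restrict motion, while the head-offset sphere has no constraint.
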PhysicsAsset /BRCosmetics/Accessories/FORT_Backpacks/Backpack_M_MED_Historian/Meshes/M_MED_Historian_Pack_Physics
bodies {
 dyn_beard_01 [{shapeType=Capsule, length=1, radius=2, center=<2.3763 0 0>, rotation=<rot 90 0 0>}]
 dyn_beard_02 [{shapeType=Capsule, length=0.75, radius=2.25, center=<2.0803 0 0>, rotation=<rot 90 0 0>}]
 dyn_beard_03 [{shapeType=Capsule, length=0.25, radius=1.75, center=<1.5924 0 -0.0001>, rotation=<rot 90 0 0>}]
 dyn_beard_04 [{shapeType=Capsule, length=0.75, radius=1.25, center=<1.5874 0 0>, rotation=<rot 90 0 0>}]
 dyn_beard_05 [{shapeType=Capsule, length=0.75, radius=1.25, center=<1.5874 0 0>, rotation=<rot 90 0 0>}]
 dyn_beard_base [{shapeType=Sphere, radius=1.125}]
 dyn_beard_parent [{shapeType=Capsule, length=1, radius=2, center=<2.1351 0 0>, rotation=<rot 90 0 0>}]
 dyn_beard_strand_01 [{shapeType=Capsule, length=0.88, radius=0.63, center=<0.8367 0 0>, rotation=<rot 90 0 0>}]
 dyn_beard_strand_02 [{shapeType=Capsule, length=1.38, radius=0.755, center=<1.1854 0 0>, rotation=<rot 90 0 0>}]
 dyn_beard_strand_03 [{shapeType=Capsule, length=0.9282, radius=0.625, center=<0.8967 0 0>, rotation=<rot 90 0 0>}]
 dyn_beard_strand_04 [{shapeType=Capsule, length=1.5391, radius=0.625, center=<1.2304 0 0>, rotation=<rot 90 0 0>}]
 dyn_head_offset [{shapeType=Sphere, radius=0.875}]
 dyn_rope_l_bottom_base [{shapeType=Sphere, radius=0.5}]
 dyn_rope_l_bottom_parent [{shapeType=Capsule, length=1.5, radius=0.875, center=<1.2656 0 -0.0001>, rotation=<rot 90 0 0>}]
 dyn_rope_l_top_base [{shapeType=Sphere, radius=0.5}]
 dyn_rope_l_top_parent [{shapeType=Capsule, length=1.5, radius=0.875, center=<1.4353 0 0>, rotation=<rot 90 0 0>}]
 dyn_rope_r_bottom_base [{shapeType=Sphere, radius=0.5}]
 dyn_rope_r_bottom_parent [{shapeType=Capsule, length=2, radius=0.88, center=<1.7303 0 0>, rotation=<rot 90 0 0>}]
 dyn_rope_r_top_base [{shapeType=Sphere, radius=0.5}]
 dyn_rope_r_top_parent [{shapeType=Capsule, length=1.6305, radius=0.8784, center=<1.3734 0 0.0001>, rotation=<rot 90 0 0>}]
 dyn_rope_top_01 [{shapeType=Capsule, length=2.2483, radius=1.0067, center=<1.402 0 0>, rotation=<rot 90 0 0>}]
 dyn_rope_top_base [{shapeType=Sphere, radius=0.5}]
 dyn_rope_top_parent [{shapeType=Capsule, length=2.0708, radius=1.3115, center=<1.4571 0 0>, rotation=<rot 90 0 0>}]
}
constraints [
 {bone1Name=dyn_rope_top_parent, bone2Name=dyn_rope_top_base, swing1LimitDegrees=15, swing2LimitDegrees=15, twistLimitDegrees=5}
 {bone1Name=dyn_rope_top_01, bone2Name=dyn_rope_top_parent, swing1LimitDegrees=15, swing2LimitDegrees=15, twistLimitDegrees=3}
 {bone1Name=dyn_rope_l_top_parent, bone2Name=dyn_rope_l_top_base, swing1LimitDegrees=20, swing2LimitDegrees=10, twistLimitDegrees=3}
 {bone1Name=dyn_rope_l_bottom_parent, bone2Name=dyn_rope_l_bottom_base, swing1LimitDegrees=15, swing2LimitDegrees=10, twistLimitDegrees=3}
 {bone1Name=dyn_rope_r_top_parent, bone2Name=dyn_rope_r_top_base, swing1LimitDegrees=20, swing2LimitDegrees=15, twistLimitDegrees=3}
 {bone1Name=dyn_rope_r_bottom_parent, bone2Name=dyn_rope_r_bottom_base, swing1LimitDegrees=10, swing2LimitDegrees=15, twistLimitDegrees=3}
 {bone1Name=dyn_beard_parent, bone2Name=dyn_beard_base, swing1LimitDegrees=10, swing2LimitDegrees=10, twistLimitDegrees=5}
 {bone1Name=dyn_beard_01, bone2Name=dyn_beard_parent, swing1LimitDegrees=10, swing2LimitDegrees=10, twistLimitDegrees=5}
 {bone1Name=dyn_beard_02, bone2Name=dyn_beard_01, swing1LimitDegrees=10, swing2LimitDegrees=10, twistLimitDegrees=5}
 {bone1Name=dyn_beard_04, bone2Name=dyn_beard_03, swing1LimitDegrees=10, swing2LimitDegrees=10, twistLimitDegrees=5}
 {bone1Name=dyn_beard_03, bone2Name=dyn_beard_02, swing1LimitDegrees=10, swing2LimitDegrees=10, twistLimitDegrees=5}
 {bone1Name=dyn_beard_05, bone2Name=dyn_beard_03, swing1LimitDegrees=10, swing2LimitDegrees=10, twistLimitDegrees=5}
 {bone1Name=dyn_beard_strand_04, bone2Name=dyn_beard_02, swing1LimitDegrees=10, swing2LimitDegrees=10, twistLimitDegrees=3}
 {bone1Name=dyn_beard_strand_03, bone2Name=dyn_beard_01, swing1LimitDegrees=10, swing2LimitDegrees=10, twistLimitDegrees=3}
 {bone1Name=dyn_beard_strand_01, bone2Name=dyn_beard_parent, swing1LimitDegrees=10, swing2LimitDegrees=10, twistLimitDegrees=3}
 {bone1Name=dyn_beard_strand_02, bone2Name=dyn_beard_parent, swing1LimitDegrees=10, swing2LimitDegrees=10, twistLimitDegrees=3}
]
